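# Niagara smoke effect for the Handlebars player emote, puffing faint tan sprites from two points offset to either side of the system.
# Gameplay spawns per unit moved, muted by User.bIsInWater, while front end uses a steady rate; front-end modifiers also adjust color and velocity.
# Puffs inherit capped velocity, get pushed along negative Y with slight lift, then slow under drag.
# Curves grow each sprite, advance its SubUV frame, fade alpha and drive dynamic material parameters.
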NiagaraSystem /BRCosmetics/Animation/Game/MainPlayer/Emotes/Handlebars/FX/NS_HandleBars_Smoke
dependencies {usedDataInterfaces [&NiagaraDataInterfaceCurve], usedDynamicInputs [], usedModules [&SystemState]}
emitters {SmokeWorld {
  emitterUpdateScript [
   &EmitterState()
   &SpawnRate(SpawnRate=&FN_FrontEndFloatModifier(Gameplay=0, Front End=40), Use Spawn Probability=false)
   &SpawnPerUnit(
    Module.Use Max Movement Threshold=true
    Module.Use Movement Tolerance=true
    Movement Tolerance=1
    Spawn Per Unit:NiagaraFloat=&Multiply_Float(
     A=&MakeCustomFloatFromBool(Bool=$User.bIsInWater, True Float=0, False Float=1)
     B=&FN_FrontEndFloatModifier(Gameplay=3, Front End=0)
    )
    Use Spawn Probability=false
   )
  ]
  particleSpawnScript [
   /Niagara/Modules/Spawn/Initialization/InitializeParticle(
    Color=&UniformRangedLinearColor(
     Minimum=&MultiplyLinearColor(
      Linear Color=<rgba 0.478 0.3791 0.2199 0.01>
      Scale Factor=&MakeVector4_FromFloat(Float=&FN_FrontEndFloatModifier(Gameplay=1, Front End=1.35))
     )
     Maximum=&MultiplyLinearColor(
      Linear Color=<rgba 0.781 0.7271 0.5498 0.06>
      Scale Factor=&MakeVector4_FromFloat(Float=&FN_FrontEndFloatModifier(Gameplay=1, Front End=1.35))
     )
    )
    Lifetime=&UniformRangedFloat(Minimum=0.5, Maximum=0.6)
    Module.Write Color=true
    Module.Write Lifetime=true
    Module.Write Mass=true
    Module.Write Material Random=true
    Module.Write Position=false
    Module.Write Scale=false
    Module.Write SpriteRotation=true
    Module.Write SpriteSize=true
    Position=&SimulationPosition()
    Sprite Rotation=&UniformRangedFloat(Minimum=0, Maximum=360)
    Sprite Size=&Vector2DFromFloat(Value=&UniformRangedFloat(Minimum=10, Maximum=20))
   )
   &TorusLocation(
    Align Torus With Axis=false
    Large Radius=3
    MaskSpawnGroup=false
    Module.Override Randomness=false
    Surface Position=1
    Torus Mode=::NewEnumerator2
    Torus Origin:NiagaraPosition=&Add_Vector(
     A=Linked() {linkedVariable {name=Particles.Position, type=NiagaraPosition}}
     B=&Multiply_VectorByFloat(Vector=$Engine.Owner.SystemZAxis, Float=0)
    )
   )
   &AccelerationForce(Acceleration=&UniformRangedVector(Minimum=<0 0 25>, Maximum=<0 0 50>))
   &PointForce(
    Force Origin=&SimulationPosition()
    Force Strength=&UniformRangedFloat(Minimum=25, Maximum=50)
    Linear Force Vector:Vector3f=<0 1 0.5>
    Module.Use Falloff=false
    Module.UseLinearForce=false
   )
   &AddVelocity(Velocity=&FN_FrontEndVector3Modifier(Gameplay=<0 -75 0>, Front End=<0 -200 0>))
   &InheritVelocity(Module.Limit Speed=true, Velocity Limit=200)
   &SystemLocation(
    MaskSpawnGroup=false
    Offset=&Lerp_Vector(A=<35 0 0>, B=<-35 0 0>, Alpha=&UniformAOrBFloat(A=0))
   ) {module=SystemLocation001}
   &ApplyInitialForces()
  ]
  particleUpdateScript [
   &ParticleState()
   &Drag(Drag=3, Use Linear Drag=true, Use Rotational Drag=false)
   &SolveForcesAndVelocity(Clamp Velocity=false, Limit Acceleration=false)
   &ScaleSpriteSize(Scale Factor=&Vector2DFromFloat(Value=&FloatFromCurve(FloatCurve=DataInterface() {propertyValues {Curve {keys [
          {arriveTangent=1.491, interpMode=RCIM_Cubic, leaveTangent=1.491, tangentMode=RCTM_User, time=-0.0046, value=1.1969}
          <at 1.0011 3.4809 RCIM_Cubic>
         ]}}})))
   /SetVariables(Particles.SubImageIndex=&FloatFromCurve(FloatCurve=DataInterface() {propertyValues {Curve {keys [
         {arriveTangent=29.8623, interpMode=RCIM_Cubic, leaveTangent=29.8623, tangentMode=RCTM_User, time=-0.0046, value=5.0556}
         <at 1.0011 22.1689 RCIM_Cubic>
        ]}}})) {isSetParameters=true}
   &DynamicMaterialParameters(
    Index 0 Param 1=&FloatFromCurve(FloatCurve=DataInterface() {propertyValues {Curve {keys [<at -0.0046 1.0113 RCIM_Cubic>,<at 1.0011 0.002>]}}})
    Index 0 Param 4=&FloatFromCurve(FloatCurve=DataInterface() {propertyValues {Curve {keys [<at 0 0 RCIM_Cubic>,<at 1 0.5>]}}})
    Module.Param0WriteEnabled=true
    Module.Param1WriteEnabled=false
    Module.Param2WriteEnabled=false
    Module.Param3WriteEnabled=true
   ) {module=DynamicMatParams}
   &CameraOffset(Camera Offset Amount=5)
   &ScaleColor(
    Scale Alpha=&FloatFromCurve(FloatCurve=DataInterface() {propertyValues {Curve {keys [
         {interpMode=RCIM_Cubic, tangentMode=RCTM_User, time=0.0006, value=0.0108}
         {arriveTangent=-0.0089, interpMode=RCIM_Cubic, leaveTangent=-0.0089, time=0.5, value=1.0113}
         <at 1.0011 0.002>
        ]}}})
    ScaleA=true
    ScaleRGB=false
   )
  ]
  rendererClasses [&NiagaraSpriteRendererProperties]
  renderers [{rendererClass=&NiagaraSpriteRendererProperties}]
 }}
summary {
 systemName=NS_HandleBars_Smoke
 emitters [{emitterName=SmokeWorld, rendererClasses [&NiagaraSpriteRendererProperties]}]
 userVariables [
  User.bIsFrontEnd {type=NiagaraBool}
  User.bIsFrontEndPreview {type=NiagaraBool}
  User.bIsInWater {type=NiagaraBool}
 ]
}
